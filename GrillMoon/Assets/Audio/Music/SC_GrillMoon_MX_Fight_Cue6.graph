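# Sound cue for the sixth fight music cue in the GrillMoon content.
# A single wave player plays the GrillMoon_MX_Fight_Cue6 sound wave with no randomization, mixing or modulation nodes.
# It routes through the Fort events music sound class with a slight volume boost and a very large max distance, and keeps playing when silent rather than stopping.
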SoundCue /GrillMoon/Assets/Audio/Music/SC_GrillMoon_MX_Fight_Cue6
root=WavePlayer_0
nodes [{id=WavePlayer_0, properties {soundWaveAssetPtr=/GrillMoon/Assets/Audio/Music/SoundWaves/GrillMoon_MX_Fight_Cue6}}]
properties {
 maxDistance=2097152
 soundClassObject=/Game/Sounds/FortSoundClassesAndModes/Fort_Events_Music
 virtualizationMode=PlayWhenSilent
 volumeMultiplier=1.1
}
