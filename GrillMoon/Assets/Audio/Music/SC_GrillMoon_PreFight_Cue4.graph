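# SC_GrillMoon_PreFight_Cue4 is a SoundCue for the fourth GrillMoon pre-fight music cue.
# A fixed three-second Delay node feeds a single Wave Player for the Cue4 music wave.
# It plays at reduced volume through the GrillMoon POI music attenuation and the Fort events music sound class.
# Its virtualization mode is PlayWhenSilent, so playback continues while inaudible.
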SoundCue /GrillMoon/Assets/Audio/Music/SC_GrillMoon_PreFight_Cue4
root=Delay_0
nodes [
 {id=Delay_0, children [WavePlayer_0], properties {delayMax=3, delayMin=3}}
 {id=WavePlayer_0, properties {soundWaveAssetPtr=/GrillMoon/Assets/Audio/Music/SoundWaves/GrillMoon_MX_PreFight_Cue4}}
]
properties {
 attenuationSettings=/GrillMoon_Audio/Assets/Settings/ATT_GrillMoon_POI_Music
 maxDistance=10500
 soundClassObject=/Game/Sounds/FortSoundClassesAndModes/Fort_Events_Music
 virtualizationMode=PlayWhenSilent
 volumeMultiplier=0.7
}
